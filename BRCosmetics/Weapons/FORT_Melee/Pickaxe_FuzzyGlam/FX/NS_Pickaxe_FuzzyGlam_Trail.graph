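# Niagara swing trail for the FuzzyGlam pickaxe cosmetic, drawing short-lived pink-magenta ribbons during a melee swing.
# A single emitter, NE_Swing_01, spawns four anim trails, each initialized from its own pair of bone or socket indices.
# Trail update and curves shape the ribbons over their life, narrowing width and fading color.
# Initial color and color-over-life both switch by time of day, with a dimmer morning fade.
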NiagaraSystem /BRCosmetics/Weapons/FORT_Melee/Pickaxe_FuzzyGlam/FX/NS_Pickaxe_FuzzyGlam_Trail
dependencies {
 usedDataInterfaces [&NiagaraDataInterfaceCurve, &NiagaraDataInterfaceColorCurve, &NiagaraDataInterfaceSocketReader]
 usedDynamicInputs [&MakeVectorFromLinearColorRGB]
 usedModules [
  &self:SystemSpawnScript.SystemScriptSource.SystemScriptGraph.NiagaraNodeAssignment_1.SetVariables_#1
  &SystemState
 ]
}
emitters {NE_Swing_01 {
  emitterSpawnScript [&FN_RibbonIDOffsetUtility()]
  emitterUpdateScript [
   &EmitterState(Life Cycle Mode=::NewEnumerator1, Loop Behavior=::NewEnumerator1, Loop Duration=0.175, UseLoopDelay=false)
   &FN_SpawnAnimTrail(Number of Anim Trails=4)
  ]
  particleSpawnScript [
   &FN_InitializeAnimTrailBoneIndecies()
   &FN_InitializeAnimTrailBoneIndecies(AnimTrail Selection=1, First Bone or Socket Index=2, Second Bone or Socket Index=3) {module=FN_InitializeAnimTrailBoneIndecies003}
   &FN_InitializeAnimTrailBoneIndecies(AnimTrail Selection=2, First Bone or Socket Index=4, Second Bone or Socket Index=5) {module=FN_InitializeAnimTrailBoneIndecies002}
   &FN_InitializeAnimTrailBoneIndecies(AnimTrail Selection=3, First Bone or Socket Index=6, Second Bone or Socket Index=7) {module=FN_InitializeAnimTrailBoneIndecies001}
   &FN_InitializeAnimTrail(
    AnimTrailWidth=0.5
    Color=&FNTimeOfDayColor(Night Color=<rgba 0.5217 0 3 1>, Morning Color=<rgba 3 0 0.2285 1>, Day Color=<rgba 3 0 1.0285 1>, Evening Color=<rgba 3 0 1.0285 1>)
    Lifetime=0.165
   )
  ]
  particleUpdateScript [
   &SolveForcesAndVelocity(Clamp Velocity=false, Limit Acceleration=false)
   &FN_UpdateAnimTrail() {module=FN_UpdateAnimTrail001}
   &ScaleRibbonWidth(Ribbon Width Scale=&FloatFromCurve(FloatCurve=DataInterface() {propertyValues {Curve {keys [{arriveTangent=-0.0042, interpMode=RCIM_Cubic, leaveTangent=-0.0042, tangentMode=RCTM_User, time=0, value=1}, <at 0.5 0>]}}}))
   &ScaleColor(
    Color Scale:LinearColor=&FNTimeOfDayColor(
     Night Color=&ColorFromCurve(ColorCurve=DataInterface() {propertyValues {
        AlphaCurve {keys [<at 0 1>,<at 0.175 0>]}
        BlueCurve {keys [<at 0 1.2963>,<at 0.25 0.2614>,<at 0.75 0.7061>,<at 1 0>]}
        GreenCurve {keys [<at 0 0>,<at 0.25 0.0029>,<at 0.75 0.6032>,<at 1 0>]}
        RedCurve {keys [<at 0 5>,<at 0.25 1>,<at 0.75 1>,<at 1 0>]}
       }})
     Morning Color=&ColorFromCurve(ColorCurve=DataInterface() {propertyValues {
        AlphaCurve {keys [<at 0 0.5>,<at 0.175 0>]}
        BlueCurve {keys [<at 0 0>,<at 0.25 0>,<at 0.75 0.7061>,<at 1 0>]}
        GreenCurve {keys [<at 0 0>,<at 0.25 0>,<at 0.75 0.6032>,<at 1 0>]}
        RedCurve {keys [<at 0 2.5>,<at 0.25 1>,<at 0.75 1>,<at 1 0>]}
       }})
     Day Color=&ColorFromCurve(ColorCurve=DataInterface() {propertyValues {
        AlphaCurve {keys [<at 0 1>,<at 0.175 0>]}
        BlueCurve {keys [<at 0 1.2963>,<at 0.25 0.2614>,<at 0.75 0.7061>,<at 1 0>]}
        GreenCurve {keys [<at 0 0>,<at 0.25 0.0029>,<at 0.75 0.6032>,<at 1 0>]}
        RedCurve {keys [<at 0 5>,<at 0.25 1>,<at 0.75 1>,<at 1 0>]}
       }})
     Evening Color=&ColorFromCurve(ColorCurve=DataInterface() {propertyValues {
        AlphaCurve {keys [<at 0 1>,<at 0.175 0>]}
        BlueCurve {keys [<at 0 1.2963>,<at 0.25 0.2614>,<at 0.75 0.7061>,<at 1 0>]}
        GreenCurve {keys [<at 0 0>,<at 0.25 0.0029>,<at 0.75 0.6032>,<at 1 0>]}
        RedCurve {keys [<at 0 5>,<at 0.25 1>,<at 0.75 1>,<at 1 0>]}
       }})
    )
    Scale Mode=3
   )
  ]
  rendererClasses [&NiagaraRibbonRendererProperties]
  renderers [{rendererClass=&NiagaraRibbonRendererProperties}]
 }}
summary {
 systemName=NS_Pickaxe_FuzzyGlam_Trail
 emitters [{emitterName=NE_Swing_01, rendererClasses [&NiagaraRibbonRendererProperties]}]
 userVariables [
  User.bIsFrontEnd {type=NiagaraBool}
  User.bIsFrontEndPreview {type=NiagaraBool}
  User.PhysicalSurface {type=NiagaraInt32}
  User.RibbonIDOffset {type=NiagaraInt32}
 ]
}
